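# Physics asset for the TroutWrist backpack cosmetic, defining bodies and joints for the pack's dynamic arm and leg bones.
# A small sphere on the pack bone anchors identically sized capsules on both arms, both legs and an extra right-leg segment.
# Angular-limited constraints tie arms and upper legs to the pack and chain the extra segment to the right leg, with narrow twist and wider swing ranges.
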PhysicsAsset /BRCosmetics/Accessories/FORT_Backpacks/Backpack_M_MED_TroutWrist/Meshes/M_MED_TroutWrist_Pack_Physics
bodies {
 dyn_arm_l [{shapeType=Capsule, length=5.5, radius=3.5, shapeName=dyn_arm_l_capsule, center=<7.3656 0.0002 0>, rotation=<rot -90 180 -179.9986>}]
 dyn_arm_r [{shapeType=Capsule, length=5.5, radius=3.5, shapeName=dyn_arm_r_capsule, center=<7.3656 -0.0001 0>, rotation=<rot -90 144.736 -144.7368>}]
 dyn_leg_l [{shapeType=Capsule, length=5.5, radius=3.5, shapeName=dyn_leg_l_capsule, center=<7.3656 0 0>, rotation=<rot -90 0 0.0002>}]
 dyn_leg_r [{shapeType=Capsule, length=5.5, radius=3.5, shapeName=dyn_leg_r_capsule, center=<7.3656 0 0>, rotation=<rot -90 0 0>}]
 dyn_leg_r_01 [{shapeType=Capsule, length=5.5, radius=3.5, shapeName=dyn_leg_r_01_capsule, center=<7.3656 0.0002 0>, rotation=<rot -90 -0.0004 0.0018>}]
 pack [{shapeType=Sphere, radius=1, shapeName=pack_sphere}]
}
constraints [
 {bone1Name=dyn_arm_r, bone2Name=pack, swing1LimitDegrees=40, swing2LimitDegrees=15, twistLimitDegrees=5}
 {bone1Name=dyn_leg_l, bone2Name=pack, swing1LimitDegrees=35, swing2LimitDegrees=15, twistLimitDegrees=5}
 {bone1Name=dyn_arm_l, bone2Name=pack, swing1LimitDegrees=40, swing2LimitDegrees=15, twistLimitDegrees=5}
 {bone1Name=dyn_leg_r, bone2Name=pack, swing1LimitDegrees=20, swing2LimitDegrees=15, twistLimitDegrees=5}
 {bone1Name=dyn_leg_r_01, bone2Name=dyn_leg_r, swing1LimitDegrees=35, swing2LimitDegrees=10, twistLimitDegrees=5}
]
